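# Material function that builds a tiled carbon-fiber surface with clear coat for vehicle cosmetics.
# Shared wrap textures drive optionally colorized base color, roughness from the normal map's green channel, and a flattened clear-coat bottom normal.
# A trim switch selects the chassis or body texture's green channel to darken base color.
# A BlendShadingModel composite uses ceiled clear-coat thickness to choose between default shading with angle-corrected normals and the full clear-coat result.
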MaterialFunction /VehicleCosmetics/SharedMaterials/MF_CarbonFiber_VC
graph {nodes [
  n1 = Multiply(A=TextureCoordinate() {uTiling=16, vTiling=16}, B=ScalarParameter() {defaultValue=1, group=CarbonFiber, parameterName=CarbonFiber__Tiling})
  n2 = TextureSample(UVs=@n1) {texture=/Game/Weapons/WeaponSkins/Wraps/Textures/T_Wrap_CarbonFiber_D}
  n3 = Multiply(A=VectorParameter() {group=CarbonFiber, parameterName=CarbonFiberColor}.RGB, B=Clamp(None=@n2.RGB) {minDefault=0.85})
  n4 = Multiply(A=@n2.RGB, B=Constant() {r=0.15})
  n5 = TextureSampleParameter2D() {
   group=Textures
   parameterName=ChassisTextures
   samplerSource=SSM_Wrap_WorldGroupSettings
   samplerType=SAMPLERTYPE_LinearColor
   sortPriority=35
   texture=/VehicleCosmetics/Textures/T_Black_Linear
   channelNames {a=Roughness, b=Metallic, g=AO-BlackOut, r=Curvature}
  }
  n6 = TextureSampleParameter2D() {
   group=Texture
   parameterName=BodyTextures
   samplerSource=SSM_Wrap_WorldGroupSettings
   samplerType=SAMPLERTYPE_LinearColor
   texture=&T_Black_Linear
  }
  n7 = TextureSample(UVs=@n1) {samplerType=SAMPLERTYPE_Normal, texture=/Game/Weapons/WeaponSkins/Wraps/Textures/T_Wrap_CarbonFiber_N}
  n8 = ScalarParameter() {defaultValue=1, group=CarbonFiber, parameterName=CarbonFiber_ClearCoat}
  n9 = SetMaterialAttributes!(
   MaterialAttributes
   BaseColor=Multiply(
    A=StaticSwitchParameter(True=@n3, False=@n4) {group=CarbonFiber, parameterName=CarbonFiber_Colorize}
    B=StaticSwitchParameter(True=@n5.AO-BlackOut, False=@n6.G) {parameterName=ApplyTrim}
   )
   Metallic=Constant() {r=0.1}
   Specular
   Roughness=Add(A=@n7.G, B=Constant() {r=0.2})
   Anisotropy
   EmissiveColor
   Opacity
   OpacityMask
   Normal
   Tangent
   WorldPositionOffset
   Displacement
   SubsurfaceColor
   CustomData0=@n8
   CustomData1=ScalarParameter() {group=CarbonFiber, parameterName=CarbonFiber_ClearCoatRoughness}
   ClearCoatBottomNormal=MaterialFunctionCall(Normal=@n7.RGB, Flatness=ScalarParameter() {defaultValue=0.25, group=CarbonFiber, parameterName=CarbonFiber_Flatten}) {materialFunction=&FlattenNormal}.Result
   ShadingModel=ShadingModel() {shadingModel=MSM_ClearCoat}
  ) {attributeSetTypes [
    #1
    #2
    #3
    #4
    #5
    #6
    #7
    #8
    #9
    #10
    #11
    #12
    #13
    #14
    #15
    #16
    #17
   ]}
  Composite!("In (MA)"=@n9, CCThickness=@n8) {subgraphName=BlendShadingModel}
  MaterialExpressionReroute_0 = Reroute(None=@n9)
  n10 = GetMaterialAttributes(None=@MaterialExpressionReroute_0."In (MA)") {attributeGetTypes [#9, #16]}
  MaterialExpressionReroute_2 = Reroute(None=@n8)
  n11 = BlendMaterialAttributes(
   A=SetMaterialAttributes!(
    MaterialAttributes=@n10.MaterialAttributes
    ShadingModel=ShadingModel()
    Normal=MaterialFunctionCall(BaseNormal=@n10.Normal, AdditionalNormal=@n10.Normal) {materialFunction=&BlendAngleCorrectedNormals}.Result
   ) {attributeSetTypes [#17, #9]}
   B=@MaterialExpressionReroute_0."In (MA)"
   Alpha=Saturate(None=Ceil(None=@MaterialExpressionReroute_2.CCThickness))
  )
  MaterialExpressionReroute_1 = Reroute(None=@n11)
  FunctionOutput(None=@MaterialExpressionReroute_1.Output)
  PinBase!(Output=@n11) {reroutePins [Output {expression=&self:MaterialExpressionReroute_1}]}
  PinBase!() {reroutePins [
    "In (MA)" {expression=&self:MaterialExpressionReroute_0}
    CCThickness {expression=&self:MaterialExpressionReroute_2}
   ]}
 ]}
