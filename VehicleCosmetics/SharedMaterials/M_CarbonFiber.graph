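# Shared carbon fiber material for vehicle cosmetics, usable on skeletal meshes and mesh deformers, with Default Lit and Clear Coat shading.
# MF_CarbonFiber_VC builds the material attributes and supplies the bottom normal for the clear coat custom output.
# Those attributes feed MF_Vehicle_FNPostFX.
# They also branch through MF_BaseVehicle_Damage into MF_VehicleDeRez_Transition, whose result nothing in the graph uses.
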Material /VehicleCosmetics/SharedMaterials/M_CarbonFiber
graph {
 properties {
  bUseMaterialAttributes=true
  bUsedWithMeshDeformer=true
  bUsedWithSkeletalMesh=true
  pixelDepthOffsetMode=PDOM_Legacy
  shadingModel=MSM_FromMaterialExpression
  usedShadingModels=Default Lit | Clear Coat
 }
 nodes [
  n1 = GetMaterialAttributes(None=MaterialFunctionCall() {materialFunction=&/MF_CarbonFiber_VC}.Result) {attributeGetTypes [#1]}
  ClearCoatNormalCustomOutput!(None=@n1.ClearCoatBottomNormal)
  MaterialFunctionCall!(In=@n1.MaterialAttributes) {materialFunction=/VehicleCosmetics/Materials/MaterialFunctions/MF_Vehicle_FNPostFX}
  MaterialFunctionCall(In=MaterialFunctionCall(Material=@n1.MaterialAttributes) {materialFunction=&MF_BaseVehicle_Damage}.Result) {materialFunction=&MF_VehicleDeRez_Transition}
 ]
}
