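# Material function that draws animated eyes and eyelids for vehicle cosmetics, outputting base color, roughness and emissive material attributes.
# Time, sine waves and a timed number generator drive eye movement, brow raises and blinks, while the Speedometer parameter triggers squinting.
# UV-gradient math shapes lids and shading, with two EyeDistance-offset samples of mask and normal textures forming fresnel-shaded irises.
# Rounded scalars toggle blink and brow movement; static switches select rust and edge gradient source.
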MaterialFunction /VehicleCosmetics/SharedMaterials/MaterialFunctions/MF_Eye_Animation
graph {nodes [
  n1 = NamedRerouteUsage()
  Composite!(None=@n1) {subgraphName=Sine}
  n2 = ScalarParameter() {
   defaultValue=1
   desc="0: No Movement, 1: Movement. 1 is default. \r\nDecimal values round and clamp to nearest 0 or 1 value."
   group=3_SpecialValues
   parameterName=BrowMovement
  }
  n3 = Multiply(A=NamedRerouteUsage(), B=Saturate(None=Round(None=@n2))) {desc=Replacing switch node.}
  Composite(None=@n3) {subgraphName=Sine}
  n4 = Multiply(A=NamedRerouteUsage(), B=NamedRerouteUsage())
  n5 = NamedRerouteUsage()
  n6 = NamedRerouteUsage()
  n7 = NamedRerouteUsage()
  n8 = ConstantBiasScale(None=NamedRerouteUsage()) {bias=-0.4, scale=2}
  n9 = LinearInterpolate(A=OneMinus(None=LinearInterpolate(A=@n7, B=@n8, Alpha=@n7)), Alpha=ConstantBiasScale(None=NamedRerouteUsage()) {bias=0.2, scale=1}) {constB=0.8}
  n10 = LinearInterpolate(A=@n4, Alpha=NamedRerouteUsage()) {constB=0}
  n11 = ScalarParameter() {defaultValue=0.5, group=3_SpecialValues, parameterName=Emissive}
  FunctionOutput(None=SetMaterialAttributes!(
    MaterialAttributes
    BaseColor=LinearInterpolate(A=@n4, B=@n5, Alpha=@n6)
    Roughness=Clamp(None=@n9) {maxDefault=0.98, minDefault=0.1}
    EmissiveColor=Multiply(A=@n10, B=@n11)
   ) {attributeSetTypes [#1, #2, #3]})
  n12 = ConstantBiasScale(None=ScalarParameter() {desc="Speedometer - uses percentage\r\nof speed from VCID", parameterName=Speedometer}) {
   bias=-0.33
   desc="Adjusted settings from Prime.\r\nLower bias == eye effect occurs at lower speeds.\r\nLower scale == slower squint tween"
   scale=97
  }
  NamedRerouteDeclaration(None=Saturate(None=@n12)) {name=Speedometer_Out, desc=Used for Alphas in Lerps, nodeColor=<rgba 1 0 0.7084 1>}
  MaterialExpressionReroute_9 = Reroute(None=@n3)
  n13 = Sine(None=@MaterialExpressionReroute_9.Input) {period=7}
  n14 = Saturate(None=ConstantBiasScale(None=Saturate(None=@n13)) {bias=-0.5, scale=10})
  MaterialExpressionReroute_10 = Reroute(None=@n14)
  n15 = Multiply(A=@n13) {constB=-1}
  n16 = Saturate(None=ConstantBiasScale(None=Saturate(None=@n15)) {bias=-0.5, scale=10})
  MaterialExpressionReroute_11 = Reroute(None=@n16)
  NamedRerouteDeclaration(None=LinearInterpolate(
    A=LinearInterpolate(Alpha=@MaterialExpressionReroute_10.Output) {constA=-0.25, constB=2.5, desc="A: Even Brows\r\nB: Right Brow Raised"}
    Alpha=@MaterialExpressionReroute_11.Output2
   ) {constB=-1.5, desc=B: Left Brow Raised}) {name=Brow_Movement_Out, nodeColor=<rgba 1 0.4508 0 1>}
  MaterialExpressionReroute_6 = Reroute(None=@n1)
  n17 = Sine(None=@MaterialExpressionReroute_6.Input) {period=7}
  n18 = Saturate(None=ConstantBiasScale(None=Saturate(None=@n17)) {bias=-0.5, scale=10})
  MaterialExpressionReroute_7 = Reroute(None=@n18)
  n19 = Multiply(A=@n17) {constB=-1}
  n20 = Saturate(None=ConstantBiasScale(None=Saturate(None=@n19)) {bias=-0.5, scale=10})
  MaterialExpressionReroute_8 = Reroute(None=@n20)
  n21 = LinearInterpolate(
   A=LinearInterpolate(Alpha=@MaterialExpressionReroute_7.Output) {constB=0.5}
   B=Multiply() {constA=-1, constB=0.5}
   Alpha=@MaterialExpressionReroute_8.Output2
  )
  NamedRerouteDeclaration(None=LinearInterpolate(A=@n21, B=ConstantBiasScale(None=@n21), Alpha=NamedRerouteUsage())) {name=Eye_Movement_Out, nodeColor=<rgba 0 1 0.9823 1>}
  NamedRerouteDeclaration(None=Time() {desc=Time Controller}) {name=Time_Out, nodeColor=<rgba 0 0.2885 1 1>}
  n22 = NamedRerouteUsage()
  n23 = AppendVector(A=ScalarParameter() {defaultValue=0.52, group=3_SpecialValues, parameterName=EyeDistance}, B=Constant())
  n24 = Add(A=@n22, B=Multiply(A=Constant() {r=-1}, B=@n23))
  n25 = TextureSampleParameter2D(UVs=@n24) {
   group=1_InputTextures
   parameterName=EyeMask_RGB
   samplerSource=SSM_Clamp_WorldGroupSettings
   texture=/VehicleCosmetics/Bodies/Pixie/Textures/T_Pixie_Eye_M
  }
  n26 = Add(A=@n22, B=@n23)
  n27 = TextureSampleParameter2D(UVs=@n26) {
   group=1_InputTextures
   parameterName=EyeMask_RGB
   samplerSource=SSM_Clamp_WorldGroupSettings
   texture=&T_Pixie_Eye_M
  }
  NamedRerouteDeclaration(None=ConstantBiasScale(None=Max(A=@n25.G, B=@n27.G)) {bias=-0.01, scale=1.2}) {name=Eye_IrisDetailTex_Out, nodeColor=<rgba 0 1 0.9823 1>}
  NamedRerouteDeclaration(None=TextureCoordinate() {desc=TexCoordinates for Lossless Gradients}) {name=TexCoordFull_Out, nodeColor=<rgba 0.9823 1 0 1>}
  n28 = NamedRerouteUsage()
  n29 = Saturate(None=ConstantBiasScale(None=@n28) {bias=-0.2, scale=0.25})
  n30 = ConstantBiasScale(None=@n28) {bias=-0.4, scale=0.2}
  n31 = Subtract(A=Subtract(A=Subtract(B=@n29), B=Saturate(None=@n30)), B=Saturate(None=ConstantBiasScale(None=@n28) {bias=-0.5, scale=0.1}))
  n32 = Saturate(None=ConstantBiasScale(None=@n28) {bias=-0.6, scale=0.75})
  n33 = ConstantBiasScale(None=@n28) {bias=-0.75, scale=0.75}
  n34 = Saturate(None=Subtract(A=Subtract(A=@n31, B=@n32), B=Saturate(None=@n33)))
  n35 = Subtract(B=Saturate(None=ConstantBiasScale(None=@n28) {bias=-0.6, scale=0.25}))
  n36 = Saturate(None=ConstantBiasScale(None=@n28) {bias=-0.65})
  n37 = ConstantBiasScale(None=@n28) {bias=-0.75, scale=1}
  n38 = Saturate(None=Subtract(A=Subtract(A=@n35, B=@n36), B=Saturate(None=@n37)))
  n39 = NamedRerouteUsage()
  n40 = TextureSampleParameter2D(UVs=NamedRerouteUsage()) {
   desc=Texture input if lossless gradient is not preferred.
   group=1_InputTextures
   parameterName=Gradient_Lossy
   texture=/VehicleCosmetics/SharedTextures/T_Body_Alpha_Whiteblank
  }
  NamedRerouteDeclaration(None=StaticSwitchParameter(True=LinearInterpolate(A=@n34, B=@n38, Alpha=@n39), False=@n40.R) {defaultValue=true, group=3_SpecialValues, parameterName=EyeEdges_UseLosslessGradient}) {name=UseLosslessGradient_Out, nodeColor=<rgba 0.9804 1 0 1>}
  n41 = ComponentMask(None=NamedRerouteUsage()) {r=true}
  n42 = LinearInterpolate(
   A=Abs(None=ConstantBiasScale(None=@n41) {bias=-0.25, scale=4})
   B=Abs(None=ConstantBiasScale(None=@n41) {bias=-0.75, scale=4})
   Alpha=Saturate(None=ConstantBiasScale(None=@n41) {bias=-0.5, scale=500000})
  )
  NamedRerouteDeclaration(None=Multiply(A=@n42, B=@n42)) {name=Brow_ShadeCyl_Out, nodeColor=<rgba 1 0.4508 0 1>}
  n43 = ConstantBiasScale(None=@n41) {bias=-0.5, scale=2}
  n44 = ConstantBiasScale(None=OneMinus(None=Abs(None=@n43))) {bias=-0.06, scale=1}
  NamedRerouteDeclaration(None=Clamp(None=ConstantBiasScale(None=@n44) {bias=-0.015, scale=25}) {minDefault=0.05}) {name=Eye_ShadeSide_Out, nodeColor=<rgba 0 1 0.9823 1>}
  NamedRerouteDeclaration(None=Power(Base=@n28) {constExponent=2.1}) {name=Brow_ShadeMidWide_Out, nodeColor=<rgba 1 0.4508 0 1>}
  n45 = Add(A=Multiply(A=NamedRerouteUsage(), B=Constant() {r=0.9}), B=Constant() {r=0.05})
  n46 = ConstantBiasScale(None=ComponentMask(None=@n45) {r=true}) {bias=-0.5, scale=2}
  NamedRerouteDeclaration(None=Saturate(None=Abs(None=@n46))) {name=Brow_ShadeMidSlim_Out, nodeColor=<rgba 1 0.4508 0 1>}
  n47 = NamedRerouteUsage()
  n48 = ConstantBiasScale(None=NamedRerouteUsage()) {bias=0.4, scale=0.15}
  n49 = LinearInterpolate(A=ConstantBiasScale(None=NamedRerouteUsage()) {bias=0, scale=0.05}, Alpha=NamedRerouteUsage()) {constB=0.05}
  n50 = ComponentMask(None=NamedRerouteUsage()) {r=true}
  n51 = Saturate(None=ConstantBiasScale(None=Add(A=@n49, B=@n50)) {bias=-0.5, scale=50})
  n52 = Multiply(A=@n51) {constB=0.0075}
  n53 = Multiply(A=@n51) {constB=-0.01}
  n54 = NamedRerouteUsage()
  n55 = ConstantBiasScale(None=NamedRerouteUsage()) {bias=0.2, scale=0.2}
  n56 = NamedRerouteUsage()
  n57 = NamedRerouteUsage()
  n58 = ConstantBiasScale(None=Saturate(None=ConstantBiasScale(None=@n57))) {bias=-0.25, scale=0.2}
  n59 = ConstantBiasScale(None=NamedRerouteUsage()) {bias=-0.925, scale=2.5}
  n60 = Multiply(A=Power(Base=Saturate(None=@n59)) {constExponent=2.25}) {constB=5}
  n61 = ConstantBiasScale(None=Subtract(A=Multiply(A=@n56) {constB=0.6}, B=Subtract(A=@n58, B=@n60))) {bias=-0.1, scale=1}
  n62 = LinearInterpolate(
   A=Add(A=Saturate(None=@n48), B=LinearInterpolate(A=@n52, B=@n53, Alpha=@n54))
   B=Add(A=Saturate(None=@n55), B=Multiply(A=@n61) {constB=0.2})
   Alpha=NamedRerouteUsage()
  )
  n63 = Subtract(A=OneMinus(None=Add(A=@n47, B=@n62)), B=NamedRerouteUsage())
  n64 = Max(A=Multiply(B=@n63) {constA=200}, B=NamedRerouteUsage())
  n65 = Saturate(None=ConstantBiasScale(None=Saturate(None=@n64)) {bias=-0.25, scale=50})
  n66 = Abs(None=@n63)
  n67 = Saturate(None=ConstantBiasScale(None=@n66) {bias=0.01, scale=20})
  n68 = NamedRerouteUsage()
  n69 = Saturate(None=ConstantBiasScale(None=Multiply(A=@n67, B=@n68)) {bias=-0.2, scale=2})
  n70 = NamedRerouteUsage() {desc=Define dark edges of eyes}
  NamedRerouteDeclaration(None=Max(A=NamedRerouteDeclaration(None=@n65) {name=Brow_MaskFullIris_Out, nodeColor=<rgba 1 0.4508 0 1>}, B=Multiply(A=@n69, B=@n70))) {name=Brow_Lines_Out, nodeColor=<rgba 1 0.4508 0 1>}
  NamedRerouteDeclaration(None=ConstantBiasScale(None=OneMinus(None=@n66)) {bias=-0.1, scale=1}) {name=Brow_ShadeUpperSoft_Out, nodeColor=<rgba 1 0.4508 0 1>}
  n71 = Fresnel() {exponent=1}
  n72 = OneMinus(None=@n71)
  n73 = LinearInterpolate(A=Saturate(None=ConstantBiasScale(None=@n72) {bias=0.125, scale=1}), Alpha=NamedRerouteUsage()) {constB=0}
  n74 = NamedRerouteUsage()
  NamedRerouteDeclaration(None=Multiply(A=Multiply(A=@n73, B=@n74) {constA=1}, B=VectorParameter() {group=2_Colors, parameterName=EyelidColor, defaultValue=<rgba 0.6795 0 0 1>}.RGB)) {name=Brow_Color_Out, nodeColor=<rgba 1 0.4508 0 1>}
  n75 = ComponentMask(None=NamedRerouteUsage()) {g=true}
  n76 = Saturate(None=NamedRerouteUsage())
  n77 = OneMinus(None=Saturate(None=NamedRerouteUsage()))
  n78 = Max(A=Multiply(A=Power(Base=@n76) {constExponent=3}) {constB=0.1}, B=Multiply(A=Power(Base=@n77) {constExponent=3}) {constB=0.05}) {desc=Blend eyelid shading}
  n79 = Subtract(A=@n75, B=ConstantBiasScale(None=Saturate(None=@n78)) {bias=-0.02, scale=1})
  n80 = LinearInterpolate(
   A=ConstantBiasScale(None=OneMinus(None=@n75)) {bias=0.01, scale=15}
   B=ConstantBiasScale(None=OneMinus(None=@n79)) {bias=-0.05, scale=12}
   Alpha=NamedRerouteUsage()
  )
  NamedRerouteDeclaration(None=Clamp(None=ConstantBiasScale(None=@n80) {bias=-0.1, scale=1.1}) {minDefault=0.25}) {name=Brow_ShadeLowerSoft_Out, nodeColor=<rgba 1 0.4508 0 1>}
  n81 = NamedRerouteUsage()
  n82 = Add(B=@n81) {constA=0.25}
  n83 = Saturate(None=ConstantBiasScale(None=Sine(None=@n82) {period=3.5}) {bias=-0.985, scale=60})
  n84 = Saturate(None=ConstantBiasScale(None=Sine(None=@n81) {period=3}) {bias=-0.985, scale=60})
  n85 = NamedRerouteUsage()
  n86 = ScalarParameter() {
   defaultValue=1
   desc="0: No Blink, 1: Blink. 1 is default. \r\nDecimal values round and clamp to nearest 0 or 1 value."
   group=3_SpecialValues
   parameterName=Blink
  }
  n87 = Round(None=@n86)
  n88 = LinearInterpolate(Alpha=Multiply(A=LinearInterpolate(A=@n83, B=@n84, Alpha=@n85), B=Saturate(None=@n87)) {desc=Replacing switch node.}) {constA=0.1, constB=-0.2}
  NamedRerouteDeclaration(None=LinearInterpolate(A=@n88, B=ConstantBiasScale(None=@n88) {bias=-0.1, scale=0.75}, Alpha=NamedRerouteUsage())) {name=Brow_Blink_Out, nodeColor=<rgba 1 0.4508 0 1>}
  n89 = LinearInterpolate(A=Saturate(None=ConstantBiasScale(None=@n75) {bias=-0.85, scale=7}), B=@n79, Alpha=NamedRerouteUsage())
  NamedRerouteDeclaration(None=Saturate(None=ConstantBiasScale(None=@n89) {bias=-0.925, scale=500})) {name=Brow_MaskLowerHard_Out, nodeColor=<rgba 1 0.4508 0 1>}
  NamedRerouteDeclaration(None=@n75) {name=TexCoordGChan_Out, nodeColor=<rgba 0.9823 1 0 1>}
  NamedRerouteDeclaration(None=@n45) {name=TexCoordRChanAltered_Out, nodeColor=<rgba 0.9823 1 0 1>}
  NamedRerouteDeclaration(None=MaterialFunctionCall!(Interval=Constant() {r=0.05}, Instances=Constant() {r=3}, ChangeSpeed=Multiply() {constA=0.2, constB=0.5}, AdditionalOffset=Constant()) {materialFunction=&/MF_TimedNumberGenerator}.Result) {name=TimedNumGen_Out, nodeColor=<rgba 1 0 0.7084 1>}
  n90 = NamedRerouteUsage()
  n91 = Sine(None=@n90) {period=30}
  n92 = AppendVector(
   A=Multiply(A=@n91) {constB=0.02}
   B=LinearInterpolate(A=Multiply(A=@n91) {constB=0.02}, B=Multiply(A=@n91) {constB=-0.02}, Alpha=Sine(None=@n90) {period=25})
  )
  n93 = Multiply(A=@n92) {constB=0.1}
  n94 = NamedRerouteUsage()
  NamedRerouteDeclaration(None=Multiply(A=LinearInterpolate(A=@n92, B=@n93, Alpha=@n94)) {constB=0.5}) {name=Eye_Easing_Out, nodeColor=<rgba 0 1 0.9823 1>}
  n95 = AppendVector(A=NamedRerouteUsage(), B=NamedRerouteUsage())
  n96 = LinearInterpolate(
   A=Add(A=Constant2Vector() {desc=Offset for iris - default open, g=-2.5, r=-1.25}, B=@n95)
   B=Add(A=Constant2Vector() {desc=Offset for iris - squinting, g=-2.5, r=-1.5}, B=Multiply(A=@n95) {constB=0.5})
   Alpha=NamedRerouteUsage()
  ) {desc=Lerp between default open and squint}
  n97 = NamedRerouteUsage()
  NamedRerouteDeclaration(None=Add(A=TextureCoordinate() {desc=Tiled Texture Coordinates, uTiling=3.5, vTiling=3.5}, B=Add(A=@n96, B=@n97))) {name=Eye_Pos_Out, nodeColor=<rgba 0 1 0.9823 1>}
  NamedRerouteDeclaration(None=@n25.R) {name=Eye_IrisR_Mask_Out, nodeColor=<rgba 0 1 0.9823 1>}
  NamedRerouteDeclaration(None=@n27.R) {name=Eye_IrisL_Mask_Out, nodeColor=<rgba 0 1 0.9823 1>}
  n98 = TextureSampleParameter2D(UVs=@n24) {
   group=1_InputTextures
   parameterName=EyeNormal
   samplerSource=SSM_Clamp_WorldGroupSettings
   samplerType=SAMPLERTYPE_Normal
   texture=/VehicleCosmetics/Bodies/Pixie/Textures/T_Pixie_Eye_N
  }
  n99 = TextureSampleParameter2D(UVs=@n26) {
   group=1_InputTextures
   parameterName=EyeNormal
   samplerSource=SSM_Clamp_WorldGroupSettings
   samplerType=SAMPLERTYPE_Normal
   texture=&T_Pixie_Eye_N
  }
  n100 = LinearInterpolate(
   A=LinearInterpolate(A=Constant3Vector() {desc=Create normal filler color for background., constant=<rgba 0 0 1 0>}, B=@n98.RGB, Alpha=NamedRerouteUsage())
   B=@n99.RGB
   Alpha=NamedRerouteUsage()
  )
  n101 = Constant() {r=0.5}
  n102 = OneMinus(None=Saturate(None=MaterialFunctionCall(Normal Vector=@n100, Power=@n101) {materialFunction=&Fresnel_Function}.Result))
  n103 = Saturate(None=ConstantBiasScale(None=@n102) {bias=0.5, scale=1})
  n104 = NamedRerouteUsage()
  n105 = LinearInterpolate(A=Saturate(None=ConstantBiasScale(None=@n102) {bias=0.5, scale=1}), B=@n103, Alpha=@n104) {desc="Blend Iris && Eyeball fresnel"}
  n106 = Multiply(A=@n104) {constB=0.2}
  n107 = Saturate(None=Multiply(A=Add(A=@n105, B=@n106)) {constB=1.2})
  n108 = VectorParameter() {group=2_Colors, parameterName=Color_Iris1, defaultValue=<rgba 0.0382 0.0612 0.2582 1>}
  n109 = VectorParameter() {group=2_Colors, parameterName=Color_Iris2, defaultValue=<rgba 0.3185 0.6867 1 1>}
  n110 = Max(A=@n25.A, B=@n27.A)
  n111 = Saturate(None=ConstantBiasScale(None=Multiply(A=@n103, B=@n110)) {bias=-0.125, scale=1.25})
  n112 = NamedRerouteDeclaration(None=Max(A=NamedRerouteUsage(), B=NamedRerouteUsage())) {name=Eye_IrisCombined_Mask_Out, nodeColor=<rgba 0 1 0.9823 1>}
  n113 = Max(A=@n25.B, B=@n27.B)
  n114 = LinearInterpolate(
   B=LinearInterpolate(B=LinearInterpolate(A=@n108.RGB, B=@n109.RGB, Alpha=@n111), Alpha=NamedRerouteUsage()) {constA=0.01}
   Alpha=Saturate(None=Subtract(A=@n112, B=@n113))
  ) {constA=0.8}
  NamedRerouteDeclaration(None=LinearInterpolate(A=Multiply(A=@n107, B=@n114), Alpha=NamedRerouteUsage()) {constA=1, constB=0.8}) {name=Eye_ColorAll_Out, nodeColor=<rgba 0 1 0.9823 1>}
  n115 = ConstantBiasScale(None=@n71) {bias=-0.1}
  n116 = Saturate(None=@n115)
  n117 = TextureSampleParameter2D() {
   group=1_InputTextures
   parameterName=RustMap
   samplerSource=SSM_Clamp_WorldGroupSettings
   texture=/VehicleCosmetics/Bodies/Tomato1/Textures/Eyes/T_Tomato_Eye_Grunge
  }
  n118 = Multiply(A=@n117.RGB, B=ScalarParameter() {defaultValue=0.4, parameterName=RustAmount})
  NamedRerouteDeclaration(None=StaticSwitchParameter(True=Add(A=@n116, B=@n118), False=@n115) {group=3_SpecialValues, parameterName=Rust}) {name=Brow_Rust_Out, nodeColor=<rgba 1 0.4508 0 1>}
  NamedRerouteDeclaration(None=@n77) {name=Eye_EdgeBase_Out, nodeColor=<rgba 1 0.9844 0 1>}
  PinBase() {reroutePins [Input {expression=&self:MaterialExpressionReroute_6}]}
  PinBase!(Output=@n18, Output2=@n20) {reroutePins [
    Output {expression=&self:MaterialExpressionReroute_7}
    Output2 {expression=&self:MaterialExpressionReroute_8}
   ]}
  PinBase!() {reroutePins [Input {expression=&self:MaterialExpressionReroute_9}]}
  PinBase!(Output=@n14, Output2=@n16) {reroutePins [
    Output {expression=&self:MaterialExpressionReroute_10}
    Output2 {expression=&self:MaterialExpressionReroute_11}
   ]}
 ]}
