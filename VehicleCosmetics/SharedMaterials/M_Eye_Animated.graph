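# M_Eye_Animated is a Default Lit material for animated eyes on vehicle cosmetics, enabled for skeletal mesh and mesh deformer use.
# It builds its material attributes in MF_Eye_Animation, then passes them through MF_VehicleDeRez_Transition, which by its name applies the vehicle de-rez transition effect.
# All shading logic lives in those two shared material functions, leaving this material as a thin wrapper.
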Material /VehicleCosmetics/SharedMaterials/M_Eye_Animated
graph {
 properties {
  bAutomaticallySetUsageInEditor=false
  bUseMaterialAttributes=true
  bUsedWithMeshDeformer=true
  bUsedWithSkeletalMesh=true
  usedShadingModels=Default Lit
 }
 nodes [MaterialFunctionCall(In=MaterialFunctionCall() {materialFunction=/VehicleCosmetics/SharedMaterials/MaterialFunctions/MF_Eye_Animation}.Result) {materialFunction=&MF_VehicleDeRez_Transition}]
}
