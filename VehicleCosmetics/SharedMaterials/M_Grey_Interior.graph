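# M_Grey_Interior is a shared vehicle cosmetics material, named for grey interior surfaces.
# Its graph holds a single Color vector parameter with a dark grey default, which material instances can override.
# It uses Default Lit shading and is flagged for skeletal mesh and mesh deformer use, with no textures or other expressions.
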Material /VehicleCosmetics/SharedMaterials/M_Grey_Interior
graph {
 properties {bUsedWithMeshDeformer=true, bUsedWithSkeletalMesh=true, pixelDepthOffsetMode=PDOM_Legacy, usedShadingModels=Default Lit}
 nodes [VectorParameter() {parameterName=Color, defaultValue=<rgba 0.0677 0.0677 0.0677 0>}]
}
